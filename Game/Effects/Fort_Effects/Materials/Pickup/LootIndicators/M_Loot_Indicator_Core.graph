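# Unlit, two-sided translucent particle material for the core of a loot pickup indicator, colored by particle color through an emissive-adjustment function.
# Opacity adds a vertically tiled, panning copy of the indicator texture's blue channel to a static copy, masked by alpha, particle alpha and view-angle fade.
# Dynamic parameters drive tiling, pan speed and fade sensitivity.
# A time-based sine scaled by speed and height parameters produces a vertical bob offset.
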Material /Game/Effects/Fort_Effects/Materials/Pickup/LootIndicators/M_Loot_Indicator_Core
graph {
 properties {
  bUsedWithParticleSprites=true
  blendMode=BLEND_Translucent
  pixelDepthOffsetMode=PDOM_Legacy
  shadingModel=MSM_Unlit
  twoSided=true
  usedShadingModels=Unlit
 }
 nodes [
  Constant() {r=0.5}
  n1 = ParticleColor()
  MaterialFunctionCall(Emissive=@n1.RGB) {materialFunction=&MF_FX_EmissiveAdjustments}
  n2 = Time()
  n3 = ScalarParameter() {defaultValue=1, parameterName=Bob_Speed}
  Multiply(A=Sine(None=Multiply(A=@n2, B=@n3)), B=VectorParameter() {parameterName=Bob_Height, defaultValue=<rgba 0 0 6 0>}.RGB)
  n4 = TextureCoordinate()
  n5 = ComponentMask(None=TextureCoordinate()) {g=true}
  n6 = DynamicParameter() {defaultValue=<rgba 0 0 0 1>, paramNames [TextureTile, PanSpeed, FadeSensitivity, Param4]}
  n7 = Panner(Coordinate=AppendVector(A=ComponentMask(None=@n4) {r=true}, B=Multiply(A=@n5, B=@n6.TextureTile)), Time=Multiply(A=Time(), B=@n6.PanSpeed)) {speedY=0.125}
  n8 = TextureSample() {texture=/Game/Effects/Fort_Effects/Textures/Icons/T_Loot_Indicator}
  n9 = Multiply(A=Add(A=TextureSample(UVs=@n7) {texture=&T_Loot_Indicator}.B, B=@n8.B), B=@n8.A)
  Multiply(
   A=Multiply(A=LinearInterpolate(Alpha=@n9) {constA=-0.01}, B=@n1.A)
   B=MaterialFunctionCall("FadeSensitivity (optional)"=@n6.FadeSensitivity) {materialFunction=&FadeBasedOnViewAngle}.Result
  )
 ]
}
